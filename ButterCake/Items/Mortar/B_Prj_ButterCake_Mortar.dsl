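; Mortar projectile for ButterCake, built on grenade projectile logic, that explodes on impact or on entering its first water body.
; Any other water entry runs a splash handler that plays a sound and spawns a water explosion effect, feeding it scale, depth and surface parameters.
; On explosion it calls the parent logic, destroys the mesh and deactivates the projectile's Niagara effect.
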(event EventBeginPlay
  (|CallParentFunction) ; This node is disabled and will not be called.\nDrag off pins to build functionality.
)

(event Collision|EventActorBeginOverlap
  (params (OtherActor :type Actor%20Object%20Reference)))

(event EventTick
  (params (DeltaSeconds :type Float%20%28single-precision%29)))

(event Collision|EventHit
  (params (MyComp :type Primitive%20Component%20Object%20Reference) (Other :type Actor%20Object%20Reference) (OtherComp :type Primitive%20Component%20Object%20Reference) (bSelfMoved :type Boolean) (HitLocation :type Vector :default "0, 0, 0") (HitNormal :type Vector :default "0, 0, 0") (NormalImpulse :type Vector :default "0, 0, 0") (Hit :type Hit%20Result%20Structure%20%28by%20ref%29))
  (Delivery|DoExplosion))

(event Game|Damage|EventAnyDamage
  (params (Damage :type Float%20%28single-precision%29) (DamageType :type Damage%20Type%20Object%20Reference) (InstigatedBy :type Controller%20Object%20Reference) (DamageCauser :type Actor%20Object%20Reference)))

(event OnEnterWaterEvent(WaterInteractionComponent)
  (params (WaterBody :type Fort%20Water%20Body%20Actor%20Object%20Reference) (WaterInteractionComponent :type Fort%20Water%20Interaction%20Component%20Object%20Reference) (bIsFirstBody :type Boolean))
  (if bIsFirstBody
    (Delivery|DoExplosion)
    (CallFunction|HandleWaterSplash WaterBody)))

(event EventOnExploded
  (params (HitActors :type Array%20of%20Actor%20Object%20References) (HitResults :type Array%20of%20Hit%20Result%20Structures))
  (|CallParentFunction HitActors HitResults)
  (Components|DestroyComponent (Variables|B_Prj_Grenade|GetMesh))
  (Components|Activation|Deactivate (Variables|Default|GetNS_Buttercake_Prj)))

(event Custom|HandleWaterSplash
  (params (WaterBody :type Fort%20Water%20Body%20Actor%20Object%20Reference))
  (bind _returnvalue (Transformation|GetActorLocation))
  (bind _outwatersurfaceinfo (WaterBody|GetWaterSurfaceInfoStruct WaterBody _returnvalue true false false))
  (bind _returnvalue_2 (Effects|Components|ParticleSystem|SpawnFXAtLocation (Utilities|ResolveSoftReference (Variables|VFX|GetNS_Explosion_Water)) _returnvalue (Variables|Default|GetExplosion%20Rotation)))
  (Audio|PlaySoundatLocation (Variables|Sounds|GetSplash%20Sound) (Transformation|GetActorLocation))
  ;[1 1 1] FX WaterInteraction params - unplugged due to waterbody info not getting pulled correctly
  (Effects|Components|ParticleSystem|SetFloatParameter _returnvalue_2 "User.EffectScale" 1.75)
  (Effects|Components|ParticleSystem|SetFloatParameter _returnvalue_2 "User.WaterDepth" (.WaterDepth (Utilities|Struct|BreakWaterSurfaceInfo _outwatersurfaceinfo)))
  (Effects|Components|ParticleSystem|SetVectorParameter _returnvalue_2 "User.WaterPlaneLocation" (Utilities|Struct|BreakWaterSurfaceInfo _outwatersurfaceinfo))
  (Effects|Components|ParticleSystem|SetVectorParameter _returnvalue_2 "User.WaterPlaneNormal" (.WaterPlaneNormal (Utilities|Struct|BreakWaterSurfaceInfo _outwatersurfaceinfo))))

(fn ConstructionScript
  (|CallParentFunction))
